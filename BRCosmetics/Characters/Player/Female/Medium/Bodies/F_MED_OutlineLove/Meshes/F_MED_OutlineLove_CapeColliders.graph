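# Physics asset of collision capsules for the OutlineLove medium female cosmetic body, named as colliders for its cape.
# Capsules cover the pelvis, spine, clavicles and six skirt bone chains (three per side), while arm and leg bodies hold no shapes.
# Constraints link each skirt chain bone to bone up to the pelvis, with swing and twist all free, so the listed angular limits do not restrict motion.
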PhysicsAsset /BRCosmetics/Characters/Player/Female/Medium/Bodies/F_MED_OutlineLove/Meshes/F_MED_OutlineLove_CapeColliders
bodies {
 calf_l []
 calf_r []
 clavicle_l [{shapeType=Capsule, length=6.5623, radius=7.5194, center=<3.9566 -1.0862 -2.1127>, rotation=<rot -77.5566 177.3915 7.7314>}]
 clavicle_r [{shapeType=Capsule, length=6.5623, radius=7.5194, center=<-3.9566 1.0862 2.1127>, rotation=<rot -77.5566 177.3915 -172.2686>}]
 dyn_skirt_l_2_02 [{shapeType=Capsule, length=18.1158, radius=7.3127, shapeName=dyn_skirt_r_3_02_capsule, center=<-1.5907 0.9745 0>, rotation=<rot 90 0 0>}]
 dyn_skirt_l_2_04 [{shapeType=Capsule, length=17.2572, radius=7.6976, shapeName=dyn_skirt_l_2_04_capsule, center=<-5.136 0 0>, rotation=<rot 90 0 0>}]
 dyn_skirt_l_2_05 [{shapeType=Capsule, length=12.8237, radius=7.6976, shapeName=dyn_skirt_l_2_05_capsule, center=<-1.5907 0 0>, rotation=<rot 90 0 0>}]
 dyn_skirt_l_3_02 [{shapeType=Capsule, length=18.1158, radius=7.3127, shapeName=dyn_skirt_r_3_02_capsule, center=<-1.5907 0.9745 0>, rotation=<rot 90 0 0>}]
 dyn_skirt_l_3_04 [{shapeType=Capsule, length=17.2572, radius=7.6976, shapeName=dyn_skirt_l_3_04_capsule, center=<-5.136 1.907 0>, rotation=<rot 90 0 0>}]
 dyn_skirt_l_3_05 [{shapeType=Capsule, length=12.8237, radius=7.6976, shapeName=dyn_skirt_l_3_05_capsule, center=<-1.5907 2.3992 0>, rotation=<rot 90 0 0>}]
 dyn_skirt_l_4_02 [{shapeType=Capsule, length=12.8237, radius=7.6976, shapeName=dyn_skirt_l_4_02_capsule, center=<8.8025 4.9467 0>, rotation=<rot 90 0 0>}]
 dyn_skirt_l_4_04 [{shapeType=Capsule, length=17.6535, radius=8.5572, shapeName=dyn_skirt_l_4_04_capsule, center=<5.2572 1.9685 0>, rotation=<rot 90 0 0>}]
 dyn_skirt_r_2_02 [{shapeType=Capsule, length=18.1158, radius=7.3127, shapeName=dyn_skirt_r_3_02_capsule, center=<-1.5907 0.9745 0>, rotation=<rot 90 0 0>}]
 dyn_skirt_r_2_04 [{shapeType=Capsule, length=17.2572, radius=7.6976, shapeName=dyn_skirt_r_2_04_capsule, center=<-5.136 0 0>, rotation=<rot 90 0 0>}]
 dyn_skirt_r_2_05 [{shapeType=Capsule, length=12.8237, radius=7.6976, shapeName=dyn_skirt_r_2_05_capsule, center=<-1.5907 0 0>, rotation=<rot 90 0 0>}]
 dyn_skirt_r_3_02 [{shapeType=Capsule, length=18.1158, radius=7.3127, shapeName=dyn_skirt_r_3_02_capsule, center=<-1.5907 0.9745 0>, rotation=<rot 90 0 0>}]
 dyn_skirt_r_3_04 [{shapeType=Capsule, length=17.2572, radius=7.6976, shapeName=dyn_skirt_r_3_04_capsule, center=<-5.136 1.603 0>, rotation=<rot 90 0 0>}]
 dyn_skirt_r_3_05 [{shapeType=Capsule, length=15.6618, radius=7.6976, shapeName=dyn_skirt_r_3_05_capsule, center=<-1.5907 1.7737 0>, rotation=<rot 90 0 0>}]
 dyn_skirt_r_4_02 [{shapeType=Capsule, length=12.8237, radius=7.6976, shapeName=dyn_skirt_r_4_02_capsule, center=<-1.5907 4.5727 0>, rotation=<rot 90 0 0>}]
 dyn_skirt_r_4_04 [{shapeType=Capsule, length=17.4635, radius=8.5572, shapeName=dyn_skirt_r_4_04_capsule, center=<-0.8789 2.3188 0>, rotation=<rot 90 0 0>}]
 lowerarm_l []
 lowerarm_r []
 pelvis [{shapeType=Capsule, length=9.6042, radius=11.1441, center=<-2.4471 -2.5289 -0.1981>}]
 spine_01 [{shapeType=Capsule, length=4.3888, radius=14.3638, center=<-8.9021 -0.0214 0>}]
 spine_02 [
  {shapeType=Capsule, length=3.5118, radius=8.7417, center=<-2.6885 -0.5337 0>}
  {shapeType=Capsule, length=20.6875, radius=9.3125, center=<-1.8317 1.9362 0>, rotation=<rot 90 0 0>}
 ]
 spine_03 [
  {shapeType=Capsule, length=6.5613, radius=9.1269, center=<0.0574 0.4416 0>}
  {shapeType=Capsule, length=21.097, radius=8.903, center=<-1.2163 0.549 0>, rotation=<rot 90 180 -175.0732>}
 ]
 spine_04 [
  {shapeType=Capsule, length=15.4197, radius=7.6556, center=<5.6234 -0.6779 0>}
  {shapeType=Capsule, length=10.0312, radius=9.2812, center=<-1.2326 1.308 0>, rotation=<rot 90 0 0>}
 ]
 spine_05 [{shapeType=Capsule, length=12.7001, radius=9.335, center=<0.6969 1.91 0>}]
 thigh_l []
 thigh_r []
 upperarm_l []
 upperarm_r []
}
constraints [
 {
  bone1Name=dyn_skirt_r_2_05
  bone2Name=dyn_skirt_r_2_04
  swing1LimitDegrees=45
  swing1Motion=Free
  swing2LimitDegrees=45
  swing2Motion=Free
  twistLimitDegrees=45
  twistMotion=Free
 }
 {
  bone1Name=dyn_skirt_l_2_02
  bone2Name=pelvis
  swing1LimitDegrees=45
  swing1Motion=Free
  swing2LimitDegrees=45
  swing2Motion=Free
  twistLimitDegrees=45
  twistMotion=Free
 }
 {
  bone1Name=dyn_skirt_l_2_04
  bone2Name=dyn_skirt_l_2_02
  swing1LimitDegrees=45
  swing1Motion=Free
  swing2LimitDegrees=45
  swing2Motion=Free
  twistLimitDegrees=45
  twistMotion=Free
 }
 {
  bone1Name=dyn_skirt_l_2_05
  bone2Name=dyn_skirt_l_2_04
  swing1LimitDegrees=45
  swing1Motion=Free
  swing2LimitDegrees=45
  swing2Motion=Free
  twistLimitDegrees=45
  twistMotion=Free
 }
 {
  bone1Name=dyn_skirt_r_3_04
  bone2Name=dyn_skirt_r_3_02
  swing1LimitDegrees=45
  swing1Motion=Free
  swing2LimitDegrees=45
  swing2Motion=Free
  twistLimitDegrees=45
  twistMotion=Free
 }
 {
  bone1Name=dyn_skirt_l_3_02
  bone2Name=pelvis
  swing1LimitDegrees=45
  swing1Motion=Free
  swing2LimitDegrees=45
  swing2Motion=Free
  twistLimitDegrees=45
  twistMotion=Free
 }
 {
  bone1Name=dyn_skirt_l_3_04
  bone2Name=dyn_skirt_l_3_02
  swing1LimitDegrees=45
  swing1Motion=Free
  swing2LimitDegrees=45
  swing2Motion=Free
  twistLimitDegrees=45
  twistMotion=Free
 }
 {
  bone1Name=dyn_skirt_l_3_05
  bone2Name=dyn_skirt_l_3_04
  swing1LimitDegrees=45
  swing1Motion=Free
  swing2LimitDegrees=45
  swing2Motion=Free
  twistLimitDegrees=45
  twistMotion=Free
 }
 {
  bone1Name=dyn_skirt_r_3_05
  bone2Name=dyn_skirt_r_3_04
  swing1LimitDegrees=45
  swing1Motion=Free
  swing2LimitDegrees=45
  swing2Motion=Free
  twistLimitDegrees=45
  twistMotion=Free
 }
 {
  bone1Name=dyn_skirt_l_4_02
  bone2Name=pelvis
  swing1LimitDegrees=45
  swing1Motion=Free
  swing2LimitDegrees=45
  swing2Motion=Free
  twistLimitDegrees=45
  twistMotion=Free
 }
 {
  bone1Name=dyn_skirt_l_4_04
  bone2Name=dyn_skirt_l_4_02
  swing1LimitDegrees=45
  swing1Motion=Free
  swing2LimitDegrees=45
  swing2Motion=Free
  twistLimitDegrees=45
  twistMotion=Free
 }
 {
  bone1Name=dyn_skirt_r_4_04
  bone2Name=dyn_skirt_r_4_02
  swing1LimitDegrees=45
  swing1Motion=Free
  swing2LimitDegrees=45
  swing2Motion=Free
  twistLimitDegrees=45
  twistMotion=Free
 }
 {
  bone1Name=dyn_skirt_r_4_02
  bone2Name=pelvis
  swing1LimitDegrees=45
  swing1Motion=Free
  swing2LimitDegrees=45
  swing2Motion=Free
  twistLimitDegrees=45
  twistMotion=Free
 }
 {
  bone1Name=dyn_skirt_r_2_02
  bone2Name=pelvis
  swing1LimitDegrees=45
  swing1Motion=Free
  swing2LimitDegrees=45
  swing2Motion=Free
  twistLimitDegrees=45
  twistMotion=Free
 }
 {
  bone1Name=dyn_skirt_r_2_04
  bone2Name=dyn_skirt_r_2_02
  swing1LimitDegrees=45
  swing1Motion=Free
  swing2LimitDegrees=45
  swing2Motion=Free
  twistLimitDegrees=45
  twistMotion=Free
 }
 {
  bone1Name=dyn_skirt_r_3_02
  bone2Name=pelvis
  swing1LimitDegrees=45
  swing1Motion=Free
  swing2LimitDegrees=45
  swing2Motion=Free
  twistLimitDegrees=45
  twistMotion=Free
 }
]
