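# Physics asset for the female medium OutlineLove player body cosmetic, covering only the lower body.
# Single capsules approximate the pelvis and each thigh, calf and foot, roughly mirrored between left and right.
# It defines no constraints, so the bodies form simple collision volumes on their bones rather than a jointed ragdoll.
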PhysicsAsset /BRCosmetics/Characters/Player/Female/Medium/Bodies/F_MED_OutlineLove/Meshes/F_MED_OutlineLove_Physics
bodies {
 calf_l [{
   shapeType=Capsule
   length=38.6806
   radius=8.8972
   shapeName=calf_l_capsule
   center=<-27.3453 -2.6695 -0.1438>
   rotation=<rot -0.2431 91.5338 -90.0684>
  }]
 calf_r [{
   shapeType=Capsule
   length=38.6806
   radius=8.8972
   shapeName=calf_r_capsule
   center=<27.3453 2.6695 0.1438>
   rotation=<rot -0.2431 91.5338 89.9316>
  }]
 foot_l [{
   shapeType=Capsule
   length=14.4982
   radius=8.5992
   shapeName=foot_l_capsule
   center=<-3.0675 7.1637 0.1159>
   rotation=<rot -1.4848 6.1684 -87.5463>
  }]
 foot_r [{
   shapeType=Capsule
   length=12.7719
   radius=8.5992
   shapeName=foot_r_capsule
   center=<3.0835 -7.3105 -0.1096>
   rotation=<rot -1.4848 6.1684 -87.5463>
  }]
 pelvis [{
   shapeType=Capsule
   length=9.3901
   radius=11.0647
   shapeName=pelvis_capsule
   center=<-6.0152 -1.9304 -0.1097>
   rotation=<rot 1.383 -90 -0.9451>
  }]
 thigh_l [{
   shapeType=Capsule
   length=35.0373
   radius=10.499
   shapeName=thigh_l_capsule
   center=<-20.9614 -1.1269 0.6614>
   rotation=<rot 0.041 -88.6098 -91.4166>
  }]
 thigh_r [{
   shapeType=Capsule
   length=37.1174
   radius=10.1723
   shapeName=thigh_r_capsule
   center=<23.3386 0.718 -0.9184>
   rotation=<rot 0.0236 91.1959 -87.9498>
  }]
}
constraints []
