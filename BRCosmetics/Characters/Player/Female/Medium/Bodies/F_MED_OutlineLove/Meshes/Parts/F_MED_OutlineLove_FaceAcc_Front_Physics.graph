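# Physics asset for the front face-accessory mesh of the OutlineLove medium female player body, covering left and right dynamic hair strands.
# Each strand is a four-capsule chain from a small master body joined to neck_01, with swing limits and tight twist limits between links.
# Capsules on neck, upper spine, clavicles and upper arms approximate the upper body around the strands.
# Master-to-neck, clavicle-to-spine and arm-to-clavicle constraints leave swing and twist free.
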PhysicsAsset /BRCosmetics/Characters/Player/Female/Medium/Bodies/F_MED_OutlineLove/Meshes/Parts/F_MED_OutlineLove_FaceAcc_Front_Physics
bodies {
 clavicle_l [{shapeType=Capsule, length=18.1429, radius=3.5, shapeName=clavicle_r_capsule, center=<2.2205 -0.109 0>, rotation=<rot 0 -83 -90>}]
 clavicle_r [{shapeType=Capsule, length=18.1429, radius=3.5, shapeName=clavicle_r_capsule, center=<-3 -0.75 0>, rotation=<rot 0 -83 -90>}]
 dyn_hair_clav_l_01 [{shapeType=Capsule, length=4.2979, radius=1.5705, shapeName=dyn_hair_clav_l_01_capsule, center=<-2.6334 0 0>, rotation=<rot 90 0 0>}]
 dyn_hair_clav_l_02 [{shapeType=Capsule, length=4.1735, radius=1.6903, shapeName=dyn_hair_clav_l_02_capsule, center=<2 0 0>, rotation=<rot 90 0 0>}]
 dyn_hair_clav_l_03 [{shapeType=Capsule, length=3.8689, radius=1.6233, shapeName=dyn_hair_clav_l_03_capsule, center=<2 0 0>, rotation=<rot 90 0 0>}]
 dyn_hair_clav_l_04 [{shapeType=Capsule, length=5.698, radius=1.8693, shapeName=dyn_hair_clav_l_04_capsule, center=<4.0605 0 0>, rotation=<rot 90 0 0>}]
 dyn_hair_clav_r_01 [{shapeType=Capsule, length=4.3168, radius=1.3711, shapeName=dyn_hair_clav_r_01_capsule, center=<3.2986 0 0>, rotation=<rot 90 0 0>}]
 dyn_hair_clav_r_02 [{shapeType=Capsule, length=5.7846, radius=1.5288, shapeName=dyn_hair_clav_r_02_capsule, center=<3.0667 0 0>, rotation=<rot 90 0 0>}]
 dyn_hair_clav_r_03 [{shapeType=Capsule, length=4.5995, radius=2.9899, shapeName=dyn_hair_clav_r_03_capsule, center=<2.7158 0 0>, rotation=<rot 90 0 0>}]
 dyn_hair_clav_r_04 [{
   shapeType=Capsule
   length=5.4009
   radius=3.4493
   shapeName=dyn_hair_clav_r_04_capsule
   center=<4.5315 0 1>
   rotation=<rot 70.0002 179.9999 180>
  }]
 dyn_hair_l_MSTR [{shapeType=Capsule, length=0.505, radius=0.505, shapeName=dyn_hair_l_MSTR_capsule, rotation=<rot 0 -90 0>}]
 dyn_hair_r_MSTR [{shapeType=Capsule, length=0.505, radius=0.505, shapeName=dyn_hair_r_MSTR_capsule, rotation=<rot 0 -90 0>}]
 neck_01 [{shapeType=Capsule, length=13.7067, radius=4.2, shapeName=neck_01_capsule, rotation=<rot 90 0 0>}]
 spine_05 [{shapeType=Capsule, length=15.5672, radius=11.9328, shapeName=spine_04_capsule, center=<-6.2006 -0.5168 0>, rotation=<rot 0 -90 0>}]
 upperarm_l [{shapeType=Capsule, length=16.5, radius=3.8, shapeName=upperarm_l_capsule, center=<5 -0.465 0>, rotation=<rot 90 2.2184 0>}]
 upperarm_r [{shapeType=Capsule, length=16.5, radius=4.3, shapeName=upperarm_l_capsule, center=<-6.0423 -0.8928 0>, rotation=<rot 90 2.2184 0>}]
}
constraints [
 {
  bone1Name=clavicle_l
  bone2Name=spine_05
  swing1LimitDegrees=45
  swing1Motion=Free
  swing2LimitDegrees=45
  swing2Motion=Free
  twistLimitDegrees=45
  twistMotion=Free
 }
 {
  bone1Name=clavicle_r
  bone2Name=spine_05
  swing1LimitDegrees=45
  swing1Motion=Free
  swing2LimitDegrees=45
  swing2Motion=Free
  twistLimitDegrees=45
  twistMotion=Free
 }
 {
  bone1Name=upperarm_l
  bone2Name=clavicle_l
  swing1LimitDegrees=45
  swing1Motion=Free
  swing2LimitDegrees=45
  swing2Motion=Free
  twistLimitDegrees=45
  twistMotion=Free
 }
 {
  bone1Name=upperarm_r
  bone2Name=clavicle_r
  swing1LimitDegrees=45
  swing1Motion=Free
  swing2LimitDegrees=45
  swing2Motion=Free
  twistLimitDegrees=45
  twistMotion=Free
 }
 {bone1Name=dyn_hair_clav_l_04, bone2Name=dyn_hair_clav_l_03, swing1LimitDegrees=45, swing2LimitDegrees=45, twistLimitDegrees=5}
 {
  bone1Name=dyn_hair_l_MSTR
  bone2Name=neck_01
  swing1LimitDegrees=45
  swing1Motion=Free
  swing2LimitDegrees=45
  swing2Motion=Free
  twistLimitDegrees=45
  twistMotion=Free
 }
 {bone1Name=dyn_hair_clav_l_01, bone2Name=dyn_hair_l_MSTR, swing1LimitDegrees=25, swing2LimitDegrees=50, twistLimitDegrees=5}
 {bone1Name=dyn_hair_clav_l_02, bone2Name=dyn_hair_clav_l_01, swing1LimitDegrees=45, swing2LimitDegrees=45, twistLimitDegrees=5}
 {bone1Name=dyn_hair_clav_l_03, bone2Name=dyn_hair_clav_l_02, swing1LimitDegrees=45, swing2LimitDegrees=45, twistLimitDegrees=5}
 {bone1Name=dyn_hair_clav_r_04, bone2Name=dyn_hair_clav_r_03, swing1LimitDegrees=45, swing2LimitDegrees=45, twistLimitDegrees=5}
 {bone1Name=dyn_hair_clav_r_02, bone2Name=dyn_hair_clav_r_01, swing1LimitDegrees=25, swing2LimitDegrees=45, twistLimitDegrees=5}
 {bone1Name=dyn_hair_clav_r_03, bone2Name=dyn_hair_clav_r_02, swing1LimitDegrees=45, swing2LimitDegrees=60, twistLimitDegrees=5}
 {
  bone1Name=dyn_hair_r_MSTR
  bone2Name=neck_01
  swing1LimitDegrees=45
  swing1Motion=Free
  swing2LimitDegrees=45
  swing2Motion=Free
  twistLimitDegrees=45
  twistMotion=Free
 }
 {bone1Name=dyn_hair_clav_r_01, bone2Name=dyn_hair_r_MSTR, swing1LimitDegrees=25, swing2LimitDegrees=45, twistLimitDegrees=5}
]
