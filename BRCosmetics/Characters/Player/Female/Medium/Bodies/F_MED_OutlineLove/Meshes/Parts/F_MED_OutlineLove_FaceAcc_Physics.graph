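# Physics asset for the OutlineLove medium female FaceAcc mesh, giving its left, right and back hair chains physical motion and collision shapes.
# Each capsule chain hangs from a free-rotating master bone on the head, with swing and twist limits set per joint.
# Free-rotation constraints also tie both side chains to the back chain mid-length.
# Capsules on head, neck, spine, pelvis, clavicles and upper arms form the body volume around the hair.
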PhysicsAsset /BRCosmetics/Characters/Player/Female/Medium/Bodies/F_MED_OutlineLove/Meshes/Parts/F_MED_OutlineLove_FaceAcc_Physics
bodies {
 clavicle_l [{
   shapeType=Capsule
   length=2.8338
   radius=6.0412
   shapeName=clavicle_l_capsule
   center=<6.7755 -0.4278 -0.9567>
   rotation=<rot 86.2624 -0.0026 -0.0026>
  }]
 clavicle_r [{
   shapeType=Capsule
   length=6.7783
   radius=6.6486
   shapeName=clavicle_l_capsule
   center=<-5.6022 0.4288 0.8801>
   rotation=<rot 86.2624 -0.0026 179.9974>
  }]
 dyn_hair_bk_00 [{shapeType=Capsule, length=3.75, radius=6.5, shapeName=dyn_hair_bk_01_capsule, center=<5 0 4>, rotation=<rot 0 0 90>}]
 dyn_hair_bk_01 [{shapeType=Capsule, length=3.25, radius=7, shapeName=dyn_hair_bk_01_capsule, center=<6 0 4>, rotation=<rot 0 0 90>}]
 dyn_hair_bk_02 [{shapeType=Capsule, length=3.125, radius=7, shapeName=dyn_hair_bk_02_capsule, center=<6 0 4>, rotation=<rot 0 0 90>}]
 dyn_hair_bk_03 [{shapeType=Capsule, length=5.375, radius=6.375, shapeName=dyn_hair_bk_03_capsule, center=<6 0 4>, rotation=<rot 4.9999 0 90>}]
 dyn_hair_bk_04 [{shapeType=Capsule, length=4.3098, radius=8.6, shapeName=dyn_hair_bk_04_capsule, center=<6 0 4.5>, rotation=<rot 0 0 90>}]
 dyn_hair_bk_MSTR [{shapeType=Capsule, length=0.505, radius=0.505, shapeName=dyn_hair_bk_MSTR_capsule, rotation=<rot 0 -90 0>}]
 dyn_hair_l_00 [{shapeType=Capsule, length=3.25, radius=3.875, shapeName=dyn_hair_l_01_capsule, center=<4 1 4>, rotation=<rot 90 0 0>}]
 dyn_hair_l_01 [{shapeType=Capsule, length=2.625, radius=4.5, shapeName=dyn_hair_l_01_capsule, center=<4 0 4>, rotation=<rot 90 0 0>}]
 dyn_hair_l_02 [
  {shapeType=Capsule, length=3, radius=6, shapeName=dyn_hair_l_02_capsule, center=<6.7 0 3.5>, rotation=<rot -0.0005 0 99.9999>}
  {
   shapeType=Capsule
   length=6.8275
   radius=2.5348
   shapeName=dyn_hair_l_02_capsule
   center=<7.591 2.0132 -1.9496>
   rotation=<rot -90 -3.5763 13.5763>
  }
 ]
 dyn_hair_l_03 [{shapeType=Capsule, length=4.625, radius=6.25, shapeName=dyn_hair_l_03_capsule, center=<5 0 4>, rotation=<rot 0 0 90>}]
 dyn_hair_l_04 [{shapeType=Capsule, length=3.6795, radius=7, shapeName=dyn_hair_l_04_capsule, center=<5 0 4>, rotation=<rot -0.0002 0 85>}]
 dyn_hair_l_MSTR [{shapeType=Capsule, length=0.505, radius=0.505, shapeName=dyn_hair_l_MSTR_capsule, rotation=<rot 0 -90 0>}]
 dyn_hair_r_00 [{shapeType=Capsule, length=3.375, radius=3.625, shapeName=dyn_hair_r_01_capsule, center=<5 0 3>, rotation=<rot 90 0 0>}]
 dyn_hair_r_01 [{shapeType=Capsule, length=2.125, radius=4.375, shapeName=dyn_hair_r_01_capsule, center=<5 0 3>, rotation=<rot 85.0002 179.9998 179.9998>}]
 dyn_hair_r_02 [{shapeType=Capsule, length=3, radius=6, shapeName=dyn_hair_l_02_capsule, center=<6.7 0 3.5>, rotation=<rot -0.0005 0 99.9999>}]
 dyn_hair_r_03 [{shapeType=Capsule, length=5.5032, radius=6.6218, shapeName=dyn_hair_r_03_capsule, center=<5 0 3>, rotation=<rot -0.0005 0 90>}]
 dyn_hair_r_04 [{shapeType=Capsule, length=12.8131, radius=7, shapeName=dyn_hair_r_04_capsule, center=<5 4.4563 3.6101>, rotation=<rot -0.0002 0 95>}]
 dyn_hair_r_MSTR [{shapeType=Capsule, length=0.505, radius=0.505, shapeName=dyn_hair_r_MSTR_capsule, rotation=<rot 0 -90 0>}]
 head [{shapeType=Capsule, length=6.1965, radius=6.5617, shapeName=head_capsule, center=<2.5967 2.5691 0>, rotation=<rot 90 0 0>}]
 neck_01 [{shapeType=Capsule, length=5.1965, radius=5.9367, shapeName=head_capsule, center=<2.5967 2.0576 0>, rotation=<rot 90 0 0>}]
 pelvis [{shapeType=Capsule, length=5.7088, radius=14.2912, shapeName=spine_05_capsule, center=<-1.553 -0.8492 0>}]
 spine_02 [{shapeType=Capsule, length=2.7088, radius=14.2912, shapeName=spine_05_capsule, center=<0.1876 5.0249 0>}]
 spine_04 [{shapeType=Capsule, length=12.4255, radius=12.0161, shapeName=spine_05_capsule, center=<1.0391 4.6466 0>}]
 spine_05 [{shapeType=Capsule, length=7.2737, radius=13.8862, shapeName=spine_05_capsule, center=<1.1451 4.5179 0>}]
 upperarm_l [{shapeType=Capsule, length=18.0838, radius=6.0412, shapeName=clavicle_l_capsule, center=<13.7543 0 -0.5227>, rotation=<rot 90 0 0>}]
 upperarm_r [{shapeType=Capsule, length=18.0838, radius=6.0412, shapeName=clavicle_l_capsule, center=<-13.7543 0 0.5227>, rotation=<rot 90 0 180>}]
}
constraints [
 {
  bone1Name=dyn_hair_l_MSTR
  bone2Name=head
  swing1LimitDegrees=45
  swing1Motion=Free
  swing2LimitDegrees=45
  swing2Motion=Free
  twistLimitDegrees=45
  twistMotion=Free
 }
 {bone1Name=dyn_hair_l_01, bone2Name=dyn_hair_l_00, swing1LimitDegrees=20, swing2LimitDegrees=20, twistLimitDegrees=10}
 {bone1Name=dyn_hair_l_02, bone2Name=dyn_hair_l_01, swing1LimitDegrees=20, swing2LimitDegrees=25, twistLimitDegrees=10}
 {bone1Name=dyn_hair_l_03, bone2Name=dyn_hair_l_02, swing1LimitDegrees=50, swing2LimitDegrees=50, twistLimitDegrees=10}
 {bone1Name=dyn_hair_l_04, bone2Name=dyn_hair_l_03, swing1LimitDegrees=20, swing2LimitDegrees=30, twistLimitDegrees=10}
 {
  bone1Name=dyn_hair_bk_MSTR
  bone2Name=head
  swing1LimitDegrees=45
  swing1Motion=Free
  swing2LimitDegrees=45
  swing2Motion=Free
  twistLimitDegrees=45
  twistMotion=Free
 }
 {bone1Name=dyn_hair_bk_01, bone2Name=dyn_hair_bk_00, swing1LimitDegrees=20, swing2LimitDegrees=20, twistLimitDegrees=10}
 {bone1Name=dyn_hair_bk_02, bone2Name=dyn_hair_bk_01, swing1LimitDegrees=20, swing2LimitDegrees=25, twistLimitDegrees=10}
 {bone1Name=dyn_hair_bk_03, bone2Name=dyn_hair_bk_02, swing1LimitDegrees=50, swing2LimitDegrees=50, twistLimitDegrees=10}
 {bone1Name=dyn_hair_bk_04, bone2Name=dyn_hair_bk_03, swing1LimitDegrees=20, swing2LimitDegrees=45, twistLimitDegrees=10}
 {
  bone1Name=dyn_hair_r_MSTR
  bone2Name=head
  swing1LimitDegrees=45
  swing1Motion=Free
  swing2LimitDegrees=45
  swing2Motion=Free
  twistLimitDegrees=45
  twistMotion=Free
 }
 {bone1Name=dyn_hair_r_01, bone2Name=dyn_hair_r_00, swing1LimitDegrees=20, swing2LimitDegrees=20, twistLimitDegrees=10}
 {bone1Name=dyn_hair_r_02, bone2Name=dyn_hair_r_01, swing1LimitDegrees=15, swing2LimitDegrees=35, twistLimitDegrees=10}
 {bone1Name=dyn_hair_r_03, bone2Name=dyn_hair_r_02, swing1LimitDegrees=50, swing2LimitDegrees=50, twistLimitDegrees=10}
 {
  bone1Name=head
  bone2Name=neck_01
  swing1LimitDegrees=45
  swing1Motion=Free
  swing2LimitDegrees=45
  swing2Motion=Free
  twistLimitDegrees=45
  twistMotion=Free
 }
 {
  bone1Name=spine_05
  bone2Name=spine_04
  swing1LimitDegrees=45
  swing1Motion=Free
  swing2LimitDegrees=45
  swing2Motion=Free
  twistLimitDegrees=45
  twistMotion=Free
 }
 {
  bone1Name=clavicle_r
  bone2Name=spine_05
  swing1LimitDegrees=45
  swing1Motion=Free
  swing2LimitDegrees=45
  swing2Motion=Free
  twistLimitDegrees=45
  twistMotion=Free
 }
 {
  bone1Name=neck_01
  bone2Name=spine_05
  swing1LimitDegrees=45
  swing1Motion=Free
  swing2LimitDegrees=45
  swing2Motion=Free
  twistLimitDegrees=45
  twistMotion=Free
 }
 {
  bone1Name=upperarm_r
  bone2Name=clavicle_r
  swing1LimitDegrees=45
  swing1Motion=Free
  swing2LimitDegrees=45
  swing2Motion=Free
  twistLimitDegrees=45
  twistMotion=Free
 }
 {
  bone1Name=spine_04
  bone2Name=spine_02
  swing1LimitDegrees=45
  swing1Motion=Free
  swing2LimitDegrees=45
  swing2Motion=Free
  twistLimitDegrees=45
  twistMotion=Free
 }
 {
  bone1Name=spine_02
  bone2Name=pelvis
  swing1LimitDegrees=45
  swing1Motion=Free
  swing2LimitDegrees=45
  swing2Motion=Free
  twistLimitDegrees=45
  twistMotion=Free
 }
 {bone1Name=dyn_hair_r_00, bone2Name=dyn_hair_r_MSTR, swing1LimitDegrees=20, swing2LimitDegrees=20, twistLimitDegrees=10}
 {bone1Name=dyn_hair_bk_00, bone2Name=dyn_hair_bk_MSTR, swing1LimitDegrees=5, swing2LimitDegrees=15, twistLimitDegrees=10}
 {bone1Name=dyn_hair_l_00, bone2Name=dyn_hair_l_MSTR, swing1LimitDegrees=20, swing2LimitDegrees=20, twistLimitDegrees=10}
 {
  bone1Name=clavicle_l
  bone2Name=spine_05
  swing1LimitDegrees=45
  swing1Motion=Free
  swing2LimitDegrees=45
  swing2Motion=Free
  twistLimitDegrees=45
  twistMotion=Free
 }
 {bone1Name=upperarm_l, bone2Name=clavicle_l, swing1LimitDegrees=20, swing2LimitDegrees=15, twistLimitDegrees=15}
 {
  bone1Name=dyn_hair_r_03
  bone2Name=dyn_hair_bk_02
  swing1LimitDegrees=45
  swing1Motion=Free
  swing2LimitDegrees=45
  swing2Motion=Free
  twistLimitDegrees=45
  twistMotion=Free
 }
 {
  bone1Name=dyn_hair_l_03
  bone2Name=dyn_hair_bk_02
  swing1LimitDegrees=45
  swing1Motion=Free
  swing2LimitDegrees=45
  swing2Motion=Free
  twistLimitDegrees=45
  twistMotion=Free
 }
 {bone1Name=dyn_hair_r_04, bone2Name=dyn_hair_r_03, swing1LimitDegrees=20, swing2LimitDegrees=40, twistLimitDegrees=10}
]
