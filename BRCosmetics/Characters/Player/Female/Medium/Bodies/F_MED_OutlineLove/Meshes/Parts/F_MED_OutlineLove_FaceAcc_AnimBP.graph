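# Animation Blueprint for the OutlineLove female face accessory, animating hair bones over the pose copied from the parent mesh.
# With backbling, back and side hair uses bone constraints and AnimDynamics; otherwise, RigidBody simulation with skydive offsets and constraints.
# A neck-based RigidBody pass follows, then emote-weighted clavicle-hair constraints, front-end idle offsets and front-strand AnimDynamics.
# The event graph derives those state weights, detects backbling and computes a parameter-asset gravity override.
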AnimBlueprint /BRCosmetics/Characters/Player/Female/Medium/Bodies/F_MED_OutlineLove/Meshes/Parts/F_MED_OutlineLove_FaceAcc_AnimBP
target_skeleton=&/F_MED_OutlineLove_FaceAcc_Skeleton
graphs [
 AnimGraph {nodes [
   n1 = RigidBody!(
    ComponentPose=Local To Component(LocalPose=Use cached pose 'Start Pose'())
    bAlphaBoolEnabled:bool
    OverrideWorldGravity:Vector="0,0,0"
    ExternalForce
    ComponentLinearAccScale:Vector="0,0,0"
    ComponentLinearVelScale:Vector="0,0,0"
   ) -> (Pose) {
    alphaInputType=Bool
    bClampLinearTranslationLimitToRefPose=true
    bOverrideWorldGravity=true
    overridePhysicsAsset=&/F_MED_OutlineLove_FaceAcc_Physics
    simulationSpace=BaseBoneSpace
    alphaBoolBlend {blendOption=Linear}
    alphaScaleBiasClamp {interpSpeedDecreasing=4, interpSpeedIncreasing=6}
    baseBoneRef {boneName=spine_05}
    componentAppliedLinearAccClamp=<3500 3500 3500>
    componentLinearAccScale=<0.6 0.6 0.2>
    componentLinearVelScale=<0.5 0.5 0.5>
    overrideWorldGravity {z=-875}
    simSpaceSettings {
     dampingAlpha=0
     maxAngularAcceleration=40
     maxAngularVelocity=8
     maxLinearAcceleration=1500
     maxLinearVelocity=700
     velocityScaleZ=0.4
     worldAlpha=0.7
    }
   }
   n2 = Get SkydiveFloat() -> (SkydiveFloat:double)
   n3 = "Transform (Modify) Bone - Bone: dyn_hair_bk_02"(ComponentPose=@n1, Alpha=@n2, Translation="0,0,0", Rotation="0,0,15", Scale="1,1,1") {rotationMode=BMM_Additive, translationMode=BMM_Additive, boneToModify {boneName=dyn_hair_bk_02}, rotation {roll=15}}
   n4 = "Transform (Modify) Bone - Bone: dyn_hair_bk_04"(
    ComponentPose="Transform (Modify) Bone - Bone: dyn_hair_l_02"(
     ComponentPose="Transform (Modify) Bone - Bone: dyn_hair_r_02"(ComponentPose=@n3, Alpha=@n2, Translation="0,0,0", Rotation="0,0,20", Scale="1,1,1") {rotationMode=BMM_Additive, translationMode=BMM_Additive, boneToModify {boneName=dyn_hair_r_02}, rotation {roll=20}}
     Alpha=@n2
     Translation="0,0,0"
     Rotation="0,0,20"
     Scale="1,1,1"
    ) {rotationMode=BMM_Additive, translationMode=BMM_Additive, boneToModify {boneName=dyn_hair_l_02}, rotation {roll=20}}
    Alpha=@n2
    Translation="0,1,0"
    Rotation="0,0,20"
    Scale="1,1,1"
   ) {rotationMode=BMM_Additive, translationMode=BMM_Additive, boneToModify {boneName=dyn_hair_bk_04}, rotation {roll=20}, translation {y=1}}
   n5 = "Constraint - dyn_hair_r_03 (PP)"!(ComponentPose=@n4, Alpha=0.5, ConstraintWeights_0=0.5, ConstraintWeights_1:float=0.5) -> (Pose) {
    alphaBoolBlend {blendOption=Linear}
    boneToModify {boneName=dyn_hair_r_03}
    constraintSetup [{targetBone {boneName=dyn_hair_bk_03}}, {targetBone {boneName=clavicle_r}}]
    constraintWeights [1,1]
   }
   n6 = "Constraint - dyn_hair_l_03 (P)"(ComponentPose=@n5, Alpha=0.5, ConstraintWeights_0=1) {
    alphaBoolBlend {blendOption=Linear}
    boneToModify {boneName=dyn_hair_l_03}
    constraintSetup [{targetBone {boneName=dyn_hair_bk_03}}]
    constraintWeights [1]
   }
   n7 = "Constraint - dyn_hair_l_04 (P)"(ComponentPose=@n6, Alpha=0.8, ConstraintWeights_0=1) {
    alphaBoolBlend {blendOption=Linear}
    boneToModify {boneName=dyn_hair_l_04}
    constraintSetup [{targetBone {boneName=dyn_hair_bk_04}}]
    constraintWeights [1]
   }
   n8 = "Constraint - dyn_hair_r_04 (P)"(ComponentPose=@n7, Alpha=0.3, ConstraintWeights_0=1) {
    alphaBoolBlend {blendOption=Linear}
    boneToModify {boneName=dyn_hair_r_04}
    constraintSetup [{targetBone {boneName=dyn_hair_bk_04}}]
    constraintWeights [1]
   }
   n9 = Component To Local(ComponentPose=@n8)
   n10 = "Constraint - dyn_hair_bk_00 (PP)"!(ComponentPose=Local To Component(LocalPose=Use cached pose 'Start Pose'()), Alpha=1, ConstraintWeights_0=0, ConstraintWeights_1:float=1) -> (Pose) {
    boneToModify {boneName=dyn_hair_bk_00}
    constraintSetup [{targetBone {boneName=spine_05}}, {targetBone {boneName=head}}]
    constraintWeights [1,1]
   }
   n11 = "Constraint - dyn_hair_bk_01 (PP)"!(ComponentPose=@n10, Alpha=1, ConstraintWeights_0=0.5, ConstraintWeights_1:float=0.5) -> (Pose) {
    boneToModify {boneName=dyn_hair_bk_01}
    constraintSetup [{targetBone {boneName=spine_05}}, {targetBone {boneName=head}}]
    constraintWeights [1,1]
   }
   n12 = "Constraint - dyn_hair_bk_02 (PP)"!(ComponentPose=@n11, Alpha=1, ConstraintWeights_0=0.6, ConstraintWeights_1:float=0.4) -> (Pose) {
    boneToModify {boneName=dyn_hair_bk_02}
    constraintSetup [{targetBone {boneName=spine_05}}, {targetBone {boneName=head}}]
    constraintWeights [1,1]
   }
   n13 = "Constraint - dyn_hair_bk_03 (P)"(ComponentPose=@n12, Alpha=1, ConstraintWeights_0=1) {boneToModify {boneName=dyn_hair_bk_03}, constraintSetup [{targetBone {boneName=spine_05}}], constraintWeights [1]}
   n14 = "Constraint - dyn_hair_bk_04 (P)"(ComponentPose=@n13, Alpha=1, ConstraintWeights_0=1) {boneToModify {boneName=dyn_hair_bk_04}, constraintSetup [{targetBone {boneName=spine_05}}], constraintWeights [1]}
   n15 = "Constraint - dyn_hair_l_00 (PP)"!(ComponentPose=@n14, Alpha=1, ConstraintWeights_0=0, ConstraintWeights_1:float=1) -> (Pose) {
    boneToModify {boneName=dyn_hair_l_00}
    constraintSetup [{targetBone {boneName=spine_05}}, {targetBone {boneName=head}}]
    constraintWeights [1,1]
   }
   n16 = "Constraint - dyn_hair_l_01 (PP)"!(ComponentPose=@n15, Alpha=1, ConstraintWeights_0=0.5, ConstraintWeights_1:float=0.5) -> (Pose) {
    boneToModify {boneName=dyn_hair_l_01}
    constraintSetup [{targetBone {boneName=spine_05}}, {targetBone {boneName=head}}]
    constraintWeights [1,1]
   }
   n17 = "Constraint - dyn_hair_l_02 (PPP)"!(ComponentPose=@n16, Alpha=1, ConstraintWeights_0=0.35, ConstraintWeights_1:float=0.3, ConstraintWeights_2:float=0.35) -> (Pose) {
    boneToModify {boneName=dyn_hair_l_02}
    constraintSetup [{targetBone {boneName=spine_05}}, {targetBone {boneName=head}}, {targetBone {boneName=clavicle_l}}]
    constraintWeights [1,1,1]
   }
   n18 = "Constraint - dyn_hair_l_03 (PP)"!(ComponentPose=@n17, Alpha=1, ConstraintWeights_0=0.3, ConstraintWeights_1:float=0.7) -> (Pose) {
    boneToModify {boneName=dyn_hair_l_03}
    constraintSetup [{targetBone {boneName=spine_05}}, {targetBone {boneName=clavicle_l}}]
    constraintWeights [1,1]
   }
   n19 = "Constraint - dyn_hair_l_04 (PP)"!(ComponentPose=@n18, Alpha=1, ConstraintWeights_0=1, ConstraintWeights_1:float=0) -> (Pose) {
    boneToModify {boneName=dyn_hair_l_04}
    constraintSetup [{targetBone {boneName=spine_05}}, {targetBone {boneName=clavicle_l}}]
    constraintWeights [1,1]
   }
   n20 = "Constraint - dyn_hair_r_00 (PP)"!(ComponentPose=@n19, Alpha=1, ConstraintWeights_0=0, ConstraintWeights_1:float=1) -> (Pose) {
    boneToModify {boneName=dyn_hair_r_00}
    constraintSetup [{targetBone {boneName=spine_05}}, {targetBone {boneName=head}}]
    constraintWeights [1,1]
   }
   n21 = "Constraint - dyn_hair_r_01 (PP)"!(ComponentPose=@n20, Alpha=1, ConstraintWeights_0=0.5, ConstraintWeights_1:float=0.5) -> (Pose) {
    boneToModify {boneName=dyn_hair_r_01}
    constraintSetup [{targetBone {boneName=spine_05}}, {targetBone {boneName=head}}]
    constraintWeights [1,1]
   }
   n22 = "Constraint - dyn_hair_r_02 (PP)"!(ComponentPose=@n21, Alpha=1, ConstraintWeights_0=0.7, ConstraintWeights_1:float=0.3) -> (Pose) {
    boneToModify {boneName=dyn_hair_r_02}
    constraintSetup [{targetBone {boneName=spine_05}}, {targetBone {boneName=head}}]
    constraintWeights [1,1]
   }
   n23 = "Constraint - dyn_hair_r_03 (PP)"(ComponentPose=@n22, Alpha=1, ConstraintWeights_0=0.5, ConstraintWeights_1=0.5) {
    boneToModify {boneName=dyn_hair_r_03}
    constraintSetup [{targetBone {boneName=spine_05}}, {targetBone {boneName=clavicle_r}}]
    constraintWeights [1,1]
   }
   n24 = "Constraint - dyn_hair_r_04 (PP)"!(ComponentPose=@n23, Alpha=1, ConstraintWeights_0=1, ConstraintWeights_1:float=0) -> (Pose) {
    boneToModify {boneName=dyn_hair_r_04}
    constraintSetup [{targetBone {boneName=spine_05}}, {targetBone {boneName=clavicle_r}}]
    constraintWeights [1,1]
   }
   n25 = "Transform (Modify) Bone - Bone: dyn_hair_bk_04"(ComponentPose=@n24, Alpha=1, Translation="0,0,-2", Rotation="0,0,0", Scale="1,1,1") {translationMode=BMM_Additive, translationSpace=BCS_BoneSpace, boneToModify {boneName=dyn_hair_bk_04}, translation {z=-2}}
   n26 = "Transform (Modify) Bone - Bone: dyn_hair_bk_03"(ComponentPose=@n25, Alpha=1, Translation="0,0,-1", Rotation="0,0,0", Scale="1,1,1") {translationMode=BMM_Additive, translationSpace=BCS_BoneSpace, boneToModify {boneName=dyn_hair_bk_03}, translation {z=-1}}
   n27 = Anim Dynamics - Bone: dyn_hair_bk_04(ComponentPose=@n26, Alpha=1, ExternalForce="0,0,0") {
    angularDampingOverride=0.9993
    angularSpringConstant=2000
    bAngularSpring=true
    bOverrideAngularDamping=true
    bUseGravityOverride=true
    linearDampingOverride=1
    simulationSpace=RootRelative
    boundBone {boneName=dyn_hair_bk_04}
    gravityOverride {z=-3000}
    physicsBodyDefinitions [{
      boundBone {boneName=dyn_hair_bk_04}
      boxExtents=<30 30 30>
      constraintSetup {angularLimitsMax=<0 0 15>, angularLimitsMin=<0 -20 -15>, angularTarget=<1 0 0>}
      localJointOffset=<2 0 0>
     }]
    simSpaceSettings {maxAngularAcceleration=1400, maxAngularVelocity=700, simSpaceAngularAlpha=1}
   }
   n28 = Anim Dynamics - Bone: dyn_hair_l_04(ComponentPose=@n27, Alpha=1, ExternalForce="0,0,0") {
    angularDampingOverride=0.9993
    angularSpringConstant=2000
    bAngularSpring=true
    bOverrideAngularDamping=true
    bUseGravityOverride=true
    linearDampingOverride=1
    simulationSpace=RootRelative
    boundBone {boneName=dyn_hair_l_04}
    gravityOverride {z=-3000}
    physicsBodyDefinitions [{
      boundBone {boneName=dyn_hair_l_04}
      boxExtents=<30 30 30>
      constraintSetup {angularLimitsMax=<0 0 20>, angularLimitsMin=<0 -20 -15>, angularTarget=<1 0 0>}
      localJointOffset=<2 0 0>
     }]
    simSpaceSettings {maxAngularAcceleration=1400, maxAngularVelocity=700, simSpaceAngularAlpha=1}
   }
   n29 = Anim Dynamics - Bone: dyn_hair_r_04(ComponentPose=@n28, Alpha=1, ExternalForce="0,0,0") {
    angularDampingOverride=0.9993
    angularSpringConstant=2000
    bAngularSpring=true
    bOverrideAngularDamping=true
    bUseGravityOverride=true
    linearDampingOverride=1
    simulationSpace=RootRelative
    boundBone {boneName=dyn_hair_r_04}
    gravityOverride {z=-3000}
    physicsBodyDefinitions [{
      boundBone {boneName=dyn_hair_r_04}
      boxExtents=<30 30 30>
      constraintSetup {angularLimitsMax=<0 0 15>, angularLimitsMin=<0 -20 -20>, angularTarget=<1 0 0>}
      localJointOffset=<2 0 0>
     }]
    simSpaceSettings {maxAngularAcceleration=1400, maxAngularVelocity=700, simSpaceAngularAlpha=1}
   }
   n30 = Component To Local(ComponentPose=@n29)
   n31 = Get isWearBackbling() -> (isWearBackbling:bool)
   n32 = RigidBody!(
    ComponentPose=Local To Component(LocalPose=Blend Poses by bool(BlendPose_0=@n30, BlendPose_1=@n9, bActiveValue=@n31))
    Alpha:float
    OverrideWorldGravity="0,0,0"
    ExternalForce
   ) -> (Pose) {
    bAllowPoseScalePropagation=false
    bOverrideWorldGravity=true
    overridePhysicsAsset=&/F_MED_OutlineLove_FaceAcc_Front_Physics
    simulationSpace=BaseBoneSpace
    baseBoneRef {boneName=neck_02}
    componentAppliedLinearAccClamp=<3000 3000 3000>
    simSpaceSettings {
     dampingAlpha=0
     maxAngularAcceleration=75
     maxAngularVelocity=40
     maxLinearAcceleration=600
     maxLinearVelocity=250
     velocityScaleZ=0.1
     worldAlpha=0.4
     externalLinearDragV=<0.3 0.3 0.4>
    }
   }
   n33 = Get Emote() -> (Emote:double)
   n34 = "Constraint - dyn_hair_clav_r_01 (PPPP)"!(
    ComponentPose=@n32
    Alpha=@n33
    ConstraintWeights_0=0
    ConstraintWeights_1:float=0.5
    ConstraintWeights_2:float=0.5
    ConstraintWeights_3:float=0
   ) -> (Pose) {
    alphaScaleBiasClamp {bInterpResult=true, interpSpeedDecreasing=7, interpSpeedIncreasing=5}
    boneToModify {boneName=dyn_hair_clav_r_01}
    constraintSetup [{targetBone {boneName=spine_04}}, {targetBone {boneName=spine_05}}, {targetBone {boneName=clavicle_r}}, {targetBone {boneName=upperarm_r}}]
    constraintWeights [1,1,1,1]
   }
   n35 = "Constraint - dyn_hair_clav_r_02 (PPP)"!(ComponentPose=@n34, Alpha=@n33, ConstraintWeights_0=0.4, ConstraintWeights_1:float=0.5, ConstraintWeights_2:float=0.1) -> (Pose) {
    alphaScaleBiasClamp {bInterpResult=true, interpSpeedDecreasing=7, interpSpeedIncreasing=5}
    boneToModify {boneName=dyn_hair_clav_r_02}
    constraintSetup [{targetBone {boneName=spine_05}}, {targetBone {boneName=clavicle_r}}, {targetBone {boneName=spine_04}}]
    constraintWeights [1,1,1]
   }
   n36 = "Constraint - dyn_hair_clav_r_03 (PPPPPP)"!(
    ComponentPose=@n35
    Alpha=@n33
    ConstraintWeights_0=0
    ConstraintWeights_1:float=0
    ConstraintWeights_2:float=0.4
    ConstraintWeights_3:float=0.4
    ConstraintWeights_4:float=0.1
    ConstraintWeights_5:float=0.1
   ) -> (Pose) {
    alphaScaleBiasClamp {bInterpResult=true, interpSpeedDecreasing=7, interpSpeedIncreasing=5}
    boneToModify {boneName=dyn_hair_clav_r_03}
    constraintSetup [
     {targetBone {boneName=spine_02}}
     {targetBone {boneName=spine_03}}
     {targetBone {boneName=spine_04}}
     {targetBone {boneName=spine_05}}
     {targetBone {boneName=clavicle_r}}
     {targetBone {boneName=upperarm_r}}
    ]
    constraintWeights [1,1,1,1,1,1]
   }
   n37 = "Constraint - dyn_hair_clav_r_04 (PPPPPP)"!(
    ComponentPose=@n36
    Alpha=@n33
    ConstraintWeights_0=0
    ConstraintWeights_1:float=0
    ConstraintWeights_2:float=0.9
    ConstraintWeights_3:float=0.1
    ConstraintWeights_4:float=0
    ConstraintWeights_5:float=0
   ) -> (Pose) {
    alphaScaleBiasClamp {bInterpResult=true, interpSpeedDecreasing=7, interpSpeedIncreasing=5}
    boneToModify {boneName=dyn_hair_clav_r_04}
    constraintSetup [
     {targetBone {boneName=spine_02}}
     {targetBone {boneName=spine_03}}
     {targetBone {boneName=spine_04}}
     {targetBone {boneName=spine_05}}
     {targetBone {boneName=clavicle_l}}
     {targetBone {boneName=upperarm_twist_01_l}}
    ]
    constraintWeights [1,1,1,1,1,1]
   }
   n38 = "Constraint - dyn_hair_clav_l_01 (PPPP)"!(
    ComponentPose=@n37
    Alpha=@n33
    ConstraintWeights_0=0
    ConstraintWeights_1:float=0.2
    ConstraintWeights_2:float=0.8
    ConstraintWeights_3:float=0
   ) -> (Pose) {
    alphaScaleBiasClamp {bInterpResult=true, interpSpeedDecreasing=7, interpSpeedIncreasing=5}
    boneToModify {boneName=dyn_hair_clav_l_01}
    constraintSetup [
     {targetBone {boneName=spine_04}}
     {targetBone {boneName=spine_05}}
     {targetBone {boneName=clavicle_l}}
     {targetBone {boneName=upperarm_twist_01_l}}
    ]
    constraintWeights [1,1,1,1]
   }
   n39 = "Constraint - dyn_hair_clav_l_02 (PPP)"!(ComponentPose=@n38, Alpha=@n33, ConstraintWeights_0=0.4, ConstraintWeights_1:float=0.5, ConstraintWeights_2:float=0.1) -> (Pose) {
    alphaScaleBiasClamp {bInterpResult=true, interpSpeedDecreasing=7, interpSpeedIncreasing=5}
    boneToModify {boneName=dyn_hair_clav_l_02}
    constraintSetup [{targetBone {boneName=spine_05}}, {targetBone {boneName=clavicle_l}}, {targetBone {boneName=spine_04}}]
    constraintWeights [1,1,1]
   }
   n40 = "Constraint - dyn_hair_clav_l_03 (PPPPPP)"!(
    ComponentPose=@n39
    Alpha=@n33
    ConstraintWeights_0=0
    ConstraintWeights_1:float=0
    ConstraintWeights_2:float=0.4
    ConstraintWeights_3:float=0.4
    ConstraintWeights_4:float=0.1
    ConstraintWeights_5:float=0.1
   ) -> (Pose) {
    alphaScaleBiasClamp {bInterpResult=true, interpSpeedDecreasing=7, interpSpeedIncreasing=5}
    boneToModify {boneName=dyn_hair_clav_l_03}
    constraintSetup [
     {targetBone {boneName=spine_02}}
     {targetBone {boneName=spine_03}}
     {targetBone {boneName=spine_04}}
     {targetBone {boneName=spine_05}}
     {targetBone {boneName=clavicle_l}}
     {targetBone {boneName=upperarm_twist_01_l}}
    ]
    constraintWeights [1,1,1,1,1,1]
   }
   n41 = "Constraint - dyn_hair_clav_l_04 (PPPPPP)"!(
    ComponentPose=@n40
    Alpha=@n33
    ConstraintWeights_0=0
    ConstraintWeights_1:float=0
    ConstraintWeights_2:float=0.9
    ConstraintWeights_3:float=0.1
    ConstraintWeights_4:float=0
    ConstraintWeights_5:float=0
   ) -> (Pose) {
    alphaScaleBiasClamp {bInterpResult=true, interpSpeedDecreasing=7, interpSpeedIncreasing=5}
    boneToModify {boneName=dyn_hair_clav_l_04}
    constraintSetup [
     {targetBone {boneName=spine_02}}
     {targetBone {boneName=spine_03}}
     {targetBone {boneName=spine_04}}
     {targetBone {boneName=spine_05}}
     {targetBone {boneName=clavicle_l}}
     {targetBone {boneName=upperarm_twist_01_l}}
    ]
    constraintWeights [1,1,1,1,1,1]
   }
   n42 = Get isIdleInFE() -> (isIdleInFE:double)
   n43 = "Transform (Modify) Bone - Bone: dyn_hair_clav_r_01"(ComponentPose=@n41, Alpha=@n42, Translation="0,0,1.7", Rotation="0,0,7", Scale="1,1,1") {rotationMode=BMM_Additive, translationMode=BMM_Additive, boneToModify {boneName=dyn_hair_clav_r_01}, rotation {roll=7}, translation {z=1.7}}
   n44 = "Transform (Modify) Bone - Bone: dyn_hair_clav_r_02"(ComponentPose=@n43, Alpha=@n42, Translation="0,0,0", Rotation="0,0,5", Scale="1,1,1") {rotationMode=BMM_Additive, translationMode=BMM_Additive, boneToModify {boneName=dyn_hair_clav_r_02}, rotation {roll=5}}
   n45 = "Transform (Modify) Bone - Bone: dyn_hair_clav_r_04"(ComponentPose=@n44, Alpha=@n42, Translation="0,0,0", Rotation="0.000,0.000,-15.000", Scale="1,1,1") {rotationMode=BMM_Additive, translationMode=BMM_Additive, boneToModify {boneName=dyn_hair_clav_r_04}, rotation {roll=-15}}
   n46 = "Transform (Modify) Bone - Bone: dyn_hair_clav_l_01"(ComponentPose=@n45, Alpha=@n42, Translation="0,0,2", Rotation="0,0,5", Scale="1,1,1") {rotationMode=BMM_Additive, translationMode=BMM_Additive, boneToModify {boneName=dyn_hair_clav_l_01}, rotation {roll=5}, translation {z=2}}
   n47 = "Transform (Modify) Bone - Bone: dyn_hair_clav_l_02"(ComponentPose=@n46, Alpha=@n42, Translation="0,0,0", Rotation="0,0,3", Scale="1,1,1") {rotationMode=BMM_Additive, translationMode=BMM_Additive, boneToModify {boneName=dyn_hair_clav_l_02}, rotation {roll=3}}
   n48 = "Transform (Modify) Bone - Bone: dyn_hair_clav_l_04"(ComponentPose=@n47, Alpha=@n42, Translation="0,0,0", Rotation="0,0,-10", Scale="1,1,1") {rotationMode=BMM_Additive, translationMode=BMM_Additive, boneToModify {boneName=dyn_hair_clav_l_04}, rotation {roll=-10}}
   n49 = Anim Dynamics - Bone: dyn_hair_front_l_01(ComponentPose=@n48, Alpha=1, ExternalForce="0,0,0") {
    angularDampingOverride=1
    bOverrideAngularDamping=true
    bOverrideLinearDamping=true
    bUseGravityOverride=true
    linearDampingOverride=0.9999
    simulationSpace=RootRelative
    boundBone {boneName=dyn_hair_front_l_01}
    componentAppliedLinearAccClamp=<3000 3000 3000>
    componentLinearAccScale=<0.1 0.1 0.1>
    componentLinearVelScale=<0.2 0.2 0.2>
    gravityOverride {z=-980}
    physicsBodyDefinitions [{
      boundBone {boneName=dyn_hair_front_l_01}
      boxExtents=<10 10 10>
      constraintSetup {angularLimitsMax=<0 90 30>, angularLimitsMin=<0 -15 0>, angularTarget=<0 0 0>}
      localJointOffset=<1 0 0>
     }]
   }
   n50 = Anim Dynamics - Bone: dyn_hair_front_r_01(ComponentPose=@n49, Alpha=1, ExternalForce="0,0,0") {
    angularDampingOverride=1
    bOverrideAngularDamping=true
    bOverrideLinearDamping=true
    bUseGravityOverride=true
    linearDampingOverride=0.9999
    simulationSpace=RootRelative
    boundBone {boneName=dyn_hair_front_r_01}
    componentAppliedLinearAccClamp=<3000 3000 3000>
    componentLinearAccScale=<0.1 0.1 0.1>
    componentLinearVelScale=<0.2 0.2 0.2>
    gravityOverride {z=-980}
    physicsBodyDefinitions [{
      boundBone {boneName=dyn_hair_front_r_01}
      boxExtents=<10 10 10>
      constraintSetup {angularLimitsMax=<0 90 0>, angularLimitsMin=<0 -5 -30>, angularTarget=<0 0 0>}
      localJointOffset=<1 0 0>
     }]
    relativeSpaceBone {boneName=dyn_hair_clav_r_02}
   }
   n51 = Anim Dynamics - Bone: dyn_hair_front_l_02(ComponentPose=@n50, Alpha=1, ExternalForce="0,0,0") {
    angularDampingOverride=1
    bOverrideAngularDamping=true
    bOverrideLinearDamping=true
    bUseGravityOverride=true
    linearDampingOverride=0.9999
    simulationSpace=RootRelative
    boundBone {boneName=dyn_hair_front_l_02}
    componentAppliedLinearAccClamp=<3000 3000 3000>
    componentLinearAccScale=<0.1 0.1 0.1>
    componentLinearVelScale=<0.2 0.2 0.2>
    gravityOverride {z=-980}
    physicsBodyDefinitions [{
      boundBone {boneName=dyn_hair_front_l_02}
      boxExtents=<10 10 10>
      constraintSetup {angularLimitsMax=<0 90 30>, angularLimitsMin=<0 -15 0>, angularTarget=<0 0 0>}
      localJointOffset=<1 0 0>
     }]
   }
   n52 = Anim Dynamics - Bone: dyn_hair_front_r_02(ComponentPose=@n51, Alpha=1, ExternalForce="0,0,0") {
    angularDampingOverride=1
    bOverrideAngularDamping=true
    bOverrideLinearDamping=true
    bUseGravityOverride=true
    linearDampingOverride=0.9999
    simulationSpace=RootRelative
    boundBone {boneName=dyn_hair_front_r_02}
    componentAppliedLinearAccClamp=<3000 3000 3000>
    componentLinearAccScale=<0.1 0.1 0.1>
    componentLinearVelScale=<0.2 0.2 0.2>
    gravityOverride {z=-980}
    physicsBodyDefinitions [{
      boundBone {boneName=dyn_hair_front_r_02}
      boxExtents=<10 10 10>
      constraintSetup {angularLimitsMax=<0 90 0>, angularLimitsMin=<0 -30 -30>, angularTarget=<0 0 0>}
      localJointOffset=<1 0 0>
     }]
    relativeSpaceBone {boneName=dyn_hair_clav_r_02}
   }
   Output Pose(Result=Component To Local(ComponentPose=@n52)) {layerGroup=None}
   Save cached pose 'Start Pose'(Pose=Copy Pose From Mesh() {bUseAttachedParent=true}) {cacheName=Start Pose}
  ]}
 EventGraph {nodes [
   Event BlueprintUpdateAnimation()
   > n1 = Sequence!(execute) -> (then_0, then_1, then_2:Exec)
   > n2 = "Gravity Override State Function (Parameter Asset) (Nativized)"(
    CharacterPartAnimInstance=Self-Reference()
    GameTimeInSeconds=GetGameTimeInSeconds()
    Parameters=&/F_MED_OutlineLoveFA_Parameters
   )
   > Set GravityOverride(GravityOverride=@n2)
   n3 = AND Boolean(A=Get isWearBackbling(), B=NOT Boolean(A=Get isWearCape() -> (isWearCape:bool)))
   n4 = Get bIsTacticalSprinting()
   Set BackPackTactJump(execute=@n1.then_1, BackPackTactJump:double=SelectFloat(A=1, B=0, bPickA=AND Boolean(A=@n4, B=@n3))) -> (Output_Get:double)
   n5 = Get isWearBackbling()
   n6 = Get bIsSkyDiving()
   > Set SkydivePack(SkydivePack:double=SelectFloat(A=1, B=0, bPickA=AND Boolean(A=@n6, B=@n5))) -> (Output_Get:double)
   n7 = Get bRigidBodyEnabled()
   n8 = Branch(
    execute=@n1.then_2
    Condition=OR Boolean!(A=Get bIsSkyDiving(), B=Get bIsCrouching(), C:bool=Get bIsPlayingEmote(), D:bool=NOT Boolean(A=@n7)) -> (ReturnValue)
   )
   > Set isIdle(isIdle:double=0) -> (Output_Get:double)
   > Set isIdleInFE(isIdleInFE:double=0) -> (Output_Get:double)
   n9 = float + float!(
    A=Get strafeLeft float() -> (strafeLeft float:double)
    B=Get strafeRight float() -> (strafeRight float:double)
    C:double=Get moveForward() -> (moveForward:double)
    D:double=Get moveBackward() -> (moveBackward:double)
   ) -> (ReturnValue)
   Set isIdle(execute=@n8.else, isIdle=Select(Option 0:double=1, Option 1:double=0, Index="float > float"(A=@n9)) -> (ReturnValue:double))
   > n10 = Branch(Condition=Get bIsInFrontEnd())
   > Set isIdleInFE(isIdleInFE=Get isIdle() -> (isIdle:double))
   Set isIdleInFE(execute=@n10.else, isIdleInFE=0)
   Event OnCharacterCustomizationLoaded()
   n11 = CheckForBackbling()
   > Set isWearBackbling(isWearBackbling:bool=@n11) -> (Output_Get:bool)
   > Set isWearCape(isWearCape:bool=@n11.bIsBackblingLong) -> (Output_Get:bool)
   Event OnBeginSkydiving()
   > Set SkydiveFloat(SkydiveFloat:double=1) -> (Output_Get:double)
   Event OnEndSkydiving()
   > Set SkydiveFloat(SkydiveFloat=0)
   Event OnBeginPlayingEmote() {is=title K2Node_Event}
   > Set Emote(Emote:double=0.5) -> (Output_Get:double)
   Event BlueprintInitializeAnimation()
   > Set Body_Mesh(Body_Mesh=GetOwnerSkeletalMeshComponent())
   Event OnEndPlayingEmote() {is=title K2Node_Event}
   > Set Emote(Emote=0)
  ]}
]
variables [
 AccScaleOverride {type=Vector}
 BackPackTactJump {type=number}
 Body_Mesh {type=SkeletalMeshComponent}
 Emote {type=number}
 Gravity Override Params {type=GravityOverrideParamsStruct}
 GravityOverride {type=Vector}
 isIdle {type=number}
 isIdleInFE {type=number}
 isWearBackbling {type=boolean}
 isWearCape {type=boolean}
 moveBackward {type=number}
 moveForward {type=number}
 SkydiveFloat {type=number}
 SkydivePack {type=number}
 strafeLeft float {type=number}
 strafeRight float {type=number}
 VelScaleOverride {type=Vector}
]
